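# Sound cue for the looping Day 03 portal ambience in the PistolCanary audio set.
# A single looping wave player plays the Day 03 portal loop wave directly as the cue root.
# It uses the shared PistolCanary portal attenuation asset, routes to the Fort_Events sound class, and plays at slightly reduced volume.
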SoundCue /PistolCanary_Audio/Assets/Portal/SC_PistolCanary_Portal_Day03
root=WavePlayer_0
nodes [{id=WavePlayer_0, properties {bLooping=true, soundWaveAssetPtr=/PistolCanary_Audio/Assets/wav/PistolCanary_Day_03_Portal_Loop}}]
properties {
 attenuationSettings=/PistolCanary_Audio/Assets/Settings/ATT_PistolCanary_Portal
 maxDistance=9600
 soundClassObject=&Fort_Events
 volumeMultiplier=0.9
}
